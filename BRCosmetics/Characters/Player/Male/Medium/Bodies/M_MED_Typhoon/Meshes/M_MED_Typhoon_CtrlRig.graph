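# Control Rig for the Typhoon male medium cosmetic body, keeping paired attachment, rod and joint bones pointed at each other.
# Execution sets left thigh and clavicle from two controls, then chains mutual AimBones, without secondary axis, across hips, torso, shoulders and arms.
# bicep_top_r targets itself, unlike the other pairs.
# Head-neck rod/wire aims, left alt-arm rotation copies (bones absent from hierarchy) and debug draws are unconnected to that chain.
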ControlRigBlueprint /BRCosmetics/Characters/Player/Male/Medium/Bodies/M_MED_Typhoon/Meshes/M_MED_Typhoon_CtrlRig
graphs [
 RigVMFunctionLibrary {links [], nodes []}
 RigVMModel {nodes [
   /Script/ControlRig.RigUnit_AimBone(
    Bone=head_rod_l
    DebugSettings=(bEnabled=false,Scale=10,WorldOffset=(Rotation=(X=0,Y=0,Z=0,W=1),Translation=(X=0,Y=0,Z=0),Scale3D=(X=1,Y=1,Z=1)))
    ExecutePin=()
    Primary=(Weight=1,Axis=(X=1,Y=0,Z=0),Target=(X=1,Y=0,Z=0),Kind=Location,Space="None")
    Secondary=(Weight=1,Axis=(X=0,Y=0,Z=0),Target=(X=1,Y=0,Z=0),Kind=Location)
    Weight=1
    bPropagateToChildren=True
    Primary.Target=&RigUnit_GetBoneTransform(Bone=neck_rod_l).Transform.Translation
   ) {is=struct RigVMUnitNode}
   > &RigUnit_AimBone(
    Bone=neck_rod_l
    DebugSettings=(Scale=10,WorldOffset=(Rotation=(X=0,Y=0,Z=0,W=1),Translation=(X=0,Y=0,Z=0),Scale3D=(X=1,Y=1,Z=1)))
    Primary=(Weight=1,Axis=(X=1,Y=0,Z=0),Target=(X=1,Y=0,Z=0),Kind=Location)
    Secondary=(Weight=1,Axis=(X=0,Y=0,Z=0),Target=(X=1,Y=0,Z=0),Kind=Location)
    Weight=1
    bPropagateToChildren=True
    Primary.Target=&RigUnit_GetBoneTransform(Bone=head_rod_l).Transform.Translation
   )
   > &RigUnit_AimBone(
    Bone=head_rod_r
    DebugSettings=(bEnabled=false,Scale=10,WorldOffset=(Rotation=(X=0,Y=0,Z=0,W=1),Translation=(X=0,Y=0,Z=0),Scale3D=(X=1,Y=1,Z=1)))
    Primary=(Weight=1,Axis=(X=1,Y=0,Z=0),Target=(X=1,Y=0,Z=0),Kind=Location,Space="None")
    Secondary=(Weight=1,Axis=(X=0,Y=0,Z=0),Target=(X=1,Y=0,Z=0),Kind=Location)
    Weight=1
    bPropagateToChildren=True
    Primary.Target=&RigUnit_GetBoneTransform(Bone=neck_rod_r).Transform.Translation
   )
   > &RigUnit_AimBone(
    Bone=neck_rod_r
    DebugSettings=(Scale=10,WorldOffset=(Rotation=(X=0,Y=0,Z=0,W=1),Translation=(X=0,Y=0,Z=0),Scale3D=(X=1,Y=1,Z=1)))
    Primary=(Weight=1,Axis=(X=1,Y=0,Z=0),Target=(X=1,Y=0,Z=0),Kind=Location)
    Secondary=(Weight=1,Axis=(X=0,Y=0,Z=0),Target=(X=1,Y=0,Z=0),Kind=Location)
    Weight=1
    bPropagateToChildren=True
    Primary.Target=&RigUnit_GetBoneTransform(Bone=head_rod_r).Transform.Translation
   )
   > &RigUnit_AimBone(
    Bone=head_wire_l
    DebugSettings=(bEnabled=false,Scale=10,WorldOffset=(Rotation=(X=0,Y=0,Z=0,W=1),Translation=(X=0,Y=0,Z=0),Scale3D=(X=1,Y=1,Z=1)))
    Primary=(Weight=1,Axis=(X=1,Y=0,Z=0),Target=(X=1,Y=0,Z=0),Kind=Location,Space="None")
    Secondary=(Weight=1,Axis=(X=0,Y=0,Z=0),Target=(X=1,Y=0,Z=0),Kind=Location)
    Weight=1
    bPropagateToChildren=True
    Primary.Target=&RigUnit_GetBoneTransform(Bone=neck_wire_l).Transform.Translation
   )
   > &RigUnit_AimBone(
    Bone=neck_wire_l
    DebugSettings=(Scale=10,WorldOffset=(Rotation=(X=0,Y=0,Z=0,W=1),Translation=(X=0,Y=0,Z=0),Scale3D=(X=1,Y=1,Z=1)))
    Primary=(Weight=1,Axis=(X=1,Y=0,Z=0),Target=(X=1,Y=0,Z=0),Kind=Location)
    Secondary=(Weight=1,Axis=(X=0,Y=0,Z=0),Target=(X=1,Y=0,Z=0),Kind=Location)
    Weight=1
    bPropagateToChildren=True
    Primary.Target=&RigUnit_GetBoneTransform(Bone=head_wire_l).Transform.Translation
   )
   > &RigUnit_AimBone(
    Bone=head_wire_r
    DebugSettings=(bEnabled=false,Scale=10,WorldOffset=(Rotation=(X=0,Y=0,Z=0,W=1),Translation=(X=0,Y=0,Z=0),Scale3D=(X=1,Y=1,Z=1)))
    Primary=(Weight=1,Axis=(X=1,Y=0,Z=0),Target=(X=1,Y=0,Z=0),Kind=Location,Space="None")
    Secondary=(Weight=1,Axis=(X=0,Y=0,Z=0),Target=(X=1,Y=0,Z=0),Kind=Location)
    Weight=1
    bPropagateToChildren=True
    Primary.Target=&RigUnit_GetBoneTransform(Bone=neck_wire_r).Transform.Translation
   )
   > &RigUnit_AimBone(
    Bone=neck_wire_r
    DebugSettings=(Scale=10,WorldOffset=(Rotation=(X=0,Y=0,Z=0,W=1),Translation=(X=0,Y=0,Z=0),Scale3D=(X=1,Y=1,Z=1)))
    Primary=(Weight=1,Axis=(X=1,Y=0,Z=0),Target=(X=1,Y=0,Z=0),Kind=Location)
    Secondary=(Weight=1,Axis=(X=0,Y=0,Z=0),Target=(X=1,Y=0,Z=0),Kind=Location)
    Weight=1
    bPropagateToChildren=True
    Primary.Target=&RigUnit_GetBoneTransform(Bone=head_wire_r).Transform.Translation
   )
   RigUnit_BeginExecution = &RigUnit_BeginExecution()
   RigUnit_GetControlTransform = /Script/ControlRig.RigUnit_GetControlTransform(Control=NewControl, Space=GlobalSpace) {is=struct RigVMUnitNode, template="GetControlValue::Execute(in Control,in Space,out Transform,out Minimum,out Maximum)"}
   RigUnit_SetBoneTransform = /Script/ControlRig.RigUnit_SetBoneTransform(
    Bone=thigh_l
    Space=GlobalSpace
    Weight=1
    bPropagateToChildren=True
    ExecutePin=@RigUnit_BeginExecution.ExecutePin
    Transform=@RigUnit_GetControlTransform.Transform
   ) {is=struct RigVMUnitNode}
   > &RigUnit_SetBoneTransform(
    Bone=clavicle_l
    Space=GlobalSpace
    Weight=1
    bPropagateToChildren=True
    Transform=&RigUnit_GetControlTransform(Control=clavicle_l_ctrl, Space=GlobalSpace) {template="GetControlValue::Execute(in Control,in Space,out Transform,out Minimum,out Maximum)"}.Transform
   )
   > &RigUnit_AimBone(
    Bone=pelvis_attach_frt_02_l
    DebugSettings=(bEnabled=false,Scale=10,WorldOffset=(Rotation=(X=0,Y=0,Z=0,W=1),Translation=(X=0,Y=0,Z=0),Scale3D=(X=1,Y=1,Z=1)))
    Primary=(Weight=1,Axis=(X=1,Y=0,Z=0),Target=(X=1,Y=0,Z=0),Kind=Location,Space="None")
    Secondary=(Weight=1,Axis=(X=0,Y=0,Z=0),Target=(X=1,Y=0,Z=0),Kind=Location)
    Weight=1
    bPropagateToChildren=True
    Primary.Target=&RigUnit_GetBoneTransform(Bone=thigh_attach_frt_01_l).Transform.Translation
   )
   > &RigUnit_AimBone(
    Bone=thigh_attach_frt_01_l
    DebugSettings=(Scale=10,WorldOffset=(Rotation=(X=0,Y=0,Z=0,W=1),Translation=(X=0,Y=0,Z=0),Scale3D=(X=1,Y=1,Z=1)))
    Primary=(Weight=1,Axis=(X=1,Y=0,Z=0),Target=(X=1,Y=0,Z=0),Kind=Location)
    Secondary=(Weight=1,Axis=(X=0,Y=0,Z=0),Target=(X=1,Y=0,Z=0),Kind=Location)
    Weight=1
    bPropagateToChildren=True
    Primary.Target=&RigUnit_GetBoneTransform(Bone=pelvis_attach_frt_02_l).Transform.Translation
   )
   > &RigUnit_AimBone(
    Bone=pelvis_attach_bck_02_l
    DebugSettings=(bEnabled=false,Scale=10,WorldOffset=(Rotation=(X=0,Y=0,Z=0,W=1),Translation=(X=0,Y=0,Z=0),Scale3D=(X=1,Y=1,Z=1)))
    Primary=(Weight=1,Axis=(X=1,Y=0,Z=0),Target=(X=1,Y=0,Z=0),Kind=Location,Space="None")
    Secondary=(Weight=1,Axis=(X=0,Y=0,Z=0),Target=(X=1,Y=0,Z=0),Kind=Location)
    Weight=1
    bPropagateToChildren=True
    Primary.Target=&RigUnit_GetBoneTransform(Bone=thigh_attach_bck_01_l).Transform.Translation
   )
   > &RigUnit_AimBone(
    Bone=thigh_attach_bck_01_l
    DebugSettings=(Scale=10,WorldOffset=(Rotation=(X=0,Y=0,Z=0,W=1),Translation=(X=0,Y=0,Z=0),Scale3D=(X=1,Y=1,Z=1)))
    Primary=(Weight=1,Axis=(X=1,Y=0,Z=0),Target=(X=1,Y=0,Z=0),Kind=Location)
    Secondary=(Weight=1,Axis=(X=0,Y=0,Z=0),Target=(X=1,Y=0,Z=0),Kind=Location)
    Weight=1
    bPropagateToChildren=True
    Primary.Target=&RigUnit_GetBoneTransform(Bone=pelvis_attach_bck_02_l).Transform.Translation
   )
   > &RigUnit_AimBone(
    Bone=armpit_top_l
    DebugSettings=(bEnabled=false,Scale=10,WorldOffset=(Rotation=(X=0,Y=0,Z=0,W=1),Translation=(X=0,Y=0,Z=0),Scale3D=(X=1,Y=1,Z=1)))
    Primary=(Weight=1,Axis=(X=1,Y=0,Z=0),Target=(X=1,Y=0,Z=0),Kind=Location,Space="None")
    Secondary=(Weight=1,Axis=(X=0,Y=0,Z=0),Target=(X=1,Y=0,Z=0),Kind=Location)
    Weight=1
    bPropagateToChildren=True
    Primary.Target=&RigUnit_GetBoneTransform(Bone=armpit_bot_l).Transform.Translation
   )
   > &RigUnit_AimBone(
    Bone=armpit_bot_l
    DebugSettings=(Scale=10,WorldOffset=(Rotation=(X=0,Y=0,Z=0,W=1),Translation=(X=0,Y=0,Z=0),Scale3D=(X=1,Y=1,Z=1)))
    Primary=(Weight=1,Axis=(X=1,Y=0,Z=0),Target=(X=1,Y=0,Z=0),Kind=Location)
    Secondary=(Weight=1,Axis=(X=0,Y=0,Z=0),Target=(X=1,Y=0,Z=0),Kind=Location)
    Weight=1
    bPropagateToChildren=True
    Primary.Target=&RigUnit_GetBoneTransform(Bone=armpit_top_l).Transform.Translation
   )
   > &RigUnit_AimBone(
    Bone=chest_top_l
    DebugSettings=(bEnabled=false,Scale=10,WorldOffset=(Rotation=(X=0,Y=0,Z=0,W=1),Translation=(X=0,Y=0,Z=0),Scale3D=(X=1,Y=1,Z=1)))
    Primary=(Weight=1,Axis=(X=1,Y=0,Z=0),Target=(X=1,Y=0,Z=0),Kind=Location,Space="None")
    Secondary=(Weight=1,Axis=(X=0,Y=0,Z=0),Target=(X=1,Y=0,Z=0),Kind=Location)
    Weight=1
    bPropagateToChildren=True
    Primary.Target=&RigUnit_GetBoneTransform(Bone=chest_bot_l).Transform.Translation
   )
   > &RigUnit_AimBone(
    Bone=chest_bot_l
    DebugSettings=(Scale=10,WorldOffset=(Rotation=(X=0,Y=0,Z=0,W=1),Translation=(X=0,Y=0,Z=0),Scale3D=(X=1,Y=1,Z=1)))
    Primary=(Weight=1,Axis=(X=1,Y=0,Z=0),Target=(X=1,Y=0,Z=0),Kind=Location)
    Secondary=(Weight=1,Axis=(X=0,Y=0,Z=0),Target=(X=1,Y=0,Z=0),Kind=Location)
    Weight=1
    bPropagateToChildren=True
    Primary.Target=&RigUnit_GetBoneTransform(Bone=chest_top_l).Transform.Translation
   )
   > &RigUnit_AimBone(
    Bone=clav_bot_l
    DebugSettings=(bEnabled=false,Scale=10,WorldOffset=(Rotation=(X=0,Y=0,Z=0,W=1),Translation=(X=0,Y=0,Z=0),Scale3D=(X=1,Y=1,Z=1)))
    Primary=(Weight=1,Axis=(X=1,Y=0,Z=0),Target=(X=1,Y=0,Z=0),Kind=Location,Space="None")
    Secondary=(Weight=1,Axis=(X=0,Y=0,Z=0),Target=(X=1,Y=0,Z=0),Kind=Location)
    Weight=1
    bPropagateToChildren=True
    Primary.Target=&RigUnit_GetBoneTransform(Bone=clav_top_l).Transform.Translation
   )
   > &RigUnit_AimBone(
    Bone=clav_top_l
    DebugSettings=(Scale=10,WorldOffset=(Rotation=(X=0,Y=0,Z=0,W=1),Translation=(X=0,Y=0,Z=0),Scale3D=(X=1,Y=1,Z=1)))
    Primary=(Weight=1,Axis=(X=1,Y=0,Z=0),Target=(X=1,Y=0,Z=0),Kind=Location)
    Secondary=(Weight=1,Axis=(X=0,Y=0,Z=0),Target=(X=1,Y=0,Z=0),Kind=Location)
    Weight=1
    bPropagateToChildren=True
    Primary.Target=&RigUnit_GetBoneTransform(Bone=clav_bot_l).Transform.Translation
   )
   > &RigUnit_AimBone(
    Bone=shld_top_l
    DebugSettings=(bEnabled=false,Scale=10,WorldOffset=(Rotation=(X=0,Y=0,Z=0,W=1),Translation=(X=0,Y=0,Z=0),Scale3D=(X=1,Y=1,Z=1)))
    Primary=(Weight=1,Axis=(X=1,Y=0,Z=0),Target=(X=1,Y=0,Z=0),Kind=Location,Space="None")
    Secondary=(Weight=1,Axis=(X=0,Y=0,Z=0),Target=(X=1,Y=0,Z=0),Kind=Location)
    Weight=1
    bPropagateToChildren=True
    Primary.Target=&RigUnit_GetBoneTransform(Bone=shld_bot_l).Transform.Translation
   )
   > &RigUnit_AimBone(
    Bone=shld_bot_l
    DebugSettings=(Scale=10,WorldOffset=(Rotation=(X=0,Y=0,Z=0,W=1),Translation=(X=0,Y=0,Z=0),Scale3D=(X=1,Y=1,Z=1)))
    Primary=(Weight=1,Axis=(X=1,Y=0,Z=0),Target=(X=1,Y=0,Z=0),Kind=Location)
    Secondary=(Weight=1,Axis=(X=0,Y=0,Z=0),Target=(X=1,Y=0,Z=0),Kind=Location)
    Weight=1
    bPropagateToChildren=True
    Primary.Target=&RigUnit_GetBoneTransform(Bone=shld_top_l).Transform.Translation
   )
   > &RigUnit_AimBone(
    Bone=elbow_top_l
    DebugSettings=(bEnabled=false,Scale=10,WorldOffset=(Rotation=(X=0,Y=0,Z=0,W=1),Translation=(X=0,Y=0,Z=0),Scale3D=(X=1,Y=1,Z=1)))
    Primary=(Weight=1,Axis=(X=1,Y=0,Z=0),Target=(X=1,Y=0,Z=0),Kind=Location,Space="None")
    Secondary=(Weight=1,Axis=(X=0,Y=0,Z=0),Target=(X=1,Y=0,Z=0),Kind=Location)
    Weight=1
    bPropagateToChildren=True
    Primary.Target=&RigUnit_GetBoneTransform(Bone=elbow_bot_l).Transform.Translation
   )
   > &RigUnit_AimBone(
    Bone=elbow_bot_l
    DebugSettings=(Scale=10,WorldOffset=(Rotation=(X=0,Y=0,Z=0,W=1),Translation=(X=0,Y=0,Z=0),Scale3D=(X=1,Y=1,Z=1)))
    Primary=(Weight=1,Axis=(X=1,Y=0,Z=0),Target=(X=1,Y=0,Z=0),Kind=Location)
    Secondary=(Weight=1,Axis=(X=0,Y=0,Z=0),Target=(X=1,Y=0,Z=0),Kind=Location)
    Weight=1
    bPropagateToChildren=True
    Primary.Target=&RigUnit_GetBoneTransform(Bone=elbow_top_l).Transform.Translation
   )
   > &RigUnit_AimBone(
    Bone=bicep_bot_l
    DebugSettings=(bEnabled=false,Scale=10,WorldOffset=(Rotation=(X=0,Y=0,Z=0,W=1),Translation=(X=0,Y=0,Z=0),Scale3D=(X=1,Y=1,Z=1)))
    Primary=(Weight=1,Axis=(X=1,Y=0,Z=0),Target=(X=1,Y=0,Z=0),Kind=Location,Space="None")
    Secondary=(Weight=1,Axis=(X=0,Y=0,Z=0),Target=(X=1,Y=0,Z=0),Kind=Location)
    Weight=1
    bPropagateToChildren=True
    Primary.Target=&RigUnit_GetBoneTransform(Bone=bicep_top_l).Transform.Translation
   )
   > &RigUnit_AimBone(
    Bone=bicep_top_l
    DebugSettings=(Scale=10,WorldOffset=(Rotation=(X=0,Y=0,Z=0,W=1),Translation=(X=0,Y=0,Z=0),Scale3D=(X=1,Y=1,Z=1)))
    Primary=(Weight=1,Axis=(X=1,Y=0,Z=0),Target=(X=1,Y=0,Z=0),Kind=Location)
    Secondary=(Weight=1,Axis=(X=0,Y=0,Z=0),Target=(X=1,Y=0,Z=0),Kind=Location)
    Weight=1
    bPropagateToChildren=True
    Primary.Target=&RigUnit_GetBoneTransform(Bone=bicep_bot_l).Transform.Translation
   )
   > &RigUnit_AimBone(
    Bone=pelvis_attach_frt_02_r
    DebugSettings=(bEnabled=false,Scale=10,WorldOffset=(Rotation=(X=0,Y=0,Z=0,W=1),Translation=(X=0,Y=0,Z=0),Scale3D=(X=1,Y=1,Z=1)))
    Primary=(Weight=1,Axis=(X=1,Y=0,Z=0),Target=(X=1,Y=0,Z=0),Kind=Location,Space="None")
    Secondary=(Weight=1,Axis=(X=0,Y=0,Z=0),Target=(X=1,Y=0,Z=0),Kind=Location)
    Weight=1
    bPropagateToChildren=True
    Primary.Target=&RigUnit_GetBoneTransform(Bone=thigh_attach_frt_01_r).Transform.Translation
   )
   > &RigUnit_AimBone(
    Bone=thigh_attach_frt_01_r
    DebugSettings=(Scale=10,WorldOffset=(Rotation=(X=0,Y=0,Z=0,W=1),Translation=(X=0,Y=0,Z=0),Scale3D=(X=1,Y=1,Z=1)))
    Primary=(Weight=1,Axis=(X=1,Y=0,Z=0),Target=(X=1,Y=0,Z=0),Kind=Location)
    Secondary=(Weight=1,Axis=(X=0,Y=0,Z=0),Target=(X=1,Y=0,Z=0),Kind=Location)
    Weight=1
    bPropagateToChildren=True
    Primary.Target=&RigUnit_GetBoneTransform(Bone=pelvis_attach_frt_02_r).Transform.Translation
   )
   > &RigUnit_AimBone(
    Bone=pelvis_attach_bck_02_r
    DebugSettings=(bEnabled=false,Scale=10,WorldOffset=(Rotation=(X=0,Y=0,Z=0,W=1),Translation=(X=0,Y=0,Z=0),Scale3D=(X=1,Y=1,Z=1)))
    Primary=(Weight=1,Axis=(X=1,Y=0,Z=0),Target=(X=1,Y=0,Z=0),Kind=Location,Space="None")
    Secondary=(Weight=1,Axis=(X=0,Y=0,Z=0),Target=(X=1,Y=0,Z=0),Kind=Location)
    Weight=1
    bPropagateToChildren=True
    Primary.Target=&RigUnit_GetBoneTransform(Bone=thigh_attach_bck_01_r).Transform.Translation
   )
   > &RigUnit_AimBone(
    Bone=thigh_attach_bck_01_r
    DebugSettings=(Scale=10,WorldOffset=(Rotation=(X=0,Y=0,Z=0,W=1),Translation=(X=0,Y=0,Z=0),Scale3D=(X=1,Y=1,Z=1)))
    Primary=(Weight=1,Axis=(X=1,Y=0,Z=0),Target=(X=1,Y=0,Z=0),Kind=Location)
    Secondary=(Weight=1,Axis=(X=0,Y=0,Z=0),Target=(X=1,Y=0,Z=0),Kind=Location)
    Weight=1
    bPropagateToChildren=True
    Primary.Target=&RigUnit_GetBoneTransform(Bone=pelvis_attach_bck_02_r).Transform.Translation
   )
   > &RigUnit_AimBone(
    Bone=armpit_top_r
    DebugSettings=(bEnabled=false,Scale=10,WorldOffset=(Rotation=(X=0,Y=0,Z=0,W=1),Translation=(X=0,Y=0,Z=0),Scale3D=(X=1,Y=1,Z=1)))
    Primary=(Weight=1,Axis=(X=1,Y=0,Z=0),Target=(X=1,Y=0,Z=0),Kind=Location,Space="None")
    Secondary=(Weight=1,Axis=(X=0,Y=0,Z=0),Target=(X=1,Y=0,Z=0),Kind=Location)
    Weight=1
    bPropagateToChildren=True
    Primary.Target=&RigUnit_GetBoneTransform(Bone=armpit_bot_r).Transform.Translation
   )
   > &RigUnit_AimBone(
    Bone=armpit_bot_r
    DebugSettings=(Scale=10,WorldOffset=(Rotation=(X=0,Y=0,Z=0,W=1),Translation=(X=0,Y=0,Z=0),Scale3D=(X=1,Y=1,Z=1)))
    Primary=(Weight=1,Axis=(X=1,Y=0,Z=0),Target=(X=1,Y=0,Z=0),Kind=Location)
    Secondary=(Weight=1,Axis=(X=0,Y=0,Z=0),Target=(X=1,Y=0,Z=0),Kind=Location)
    Weight=1
    bPropagateToChildren=True
    Primary.Target=&RigUnit_GetBoneTransform(Bone=armpit_top_r).Transform.Translation
   )
   > &RigUnit_AimBone(
    Bone=chest_top_r
    DebugSettings=(bEnabled=false,Scale=10,WorldOffset=(Rotation=(X=0,Y=0,Z=0,W=1),Translation=(X=0,Y=0,Z=0),Scale3D=(X=1,Y=1,Z=1)))
    Primary=(Weight=1,Axis=(X=1,Y=0,Z=0),Target=(X=1,Y=0,Z=0),Kind=Location,Space="None")
    Secondary=(Weight=1,Axis=(X=0,Y=0,Z=0),Target=(X=1,Y=0,Z=0),Kind=Location)
    Weight=1
    bPropagateToChildren=True
    Primary.Target=&RigUnit_GetBoneTransform(Bone=chest_bot_r).Transform.Translation
   )
   > &RigUnit_AimBone(
    Bone=chest_bot_r
    DebugSettings=(Scale=10,WorldOffset=(Rotation=(X=0,Y=0,Z=0,W=1),Translation=(X=0,Y=0,Z=0),Scale3D=(X=1,Y=1,Z=1)))
    Primary=(Weight=1,Axis=(X=1,Y=0,Z=0),Target=(X=1,Y=0,Z=0),Kind=Location)
    Secondary=(Weight=1,Axis=(X=0,Y=0,Z=0),Target=(X=1,Y=0,Z=0),Kind=Location)
    Weight=1
    bPropagateToChildren=True
    Primary.Target=&RigUnit_GetBoneTransform(Bone=chest_top_r).Transform.Translation
   )
   > &RigUnit_AimBone(
    Bone=clav_bot_r
    DebugSettings=(bEnabled=false,Scale=10,WorldOffset=(Rotation=(X=0,Y=0,Z=0,W=1),Translation=(X=0,Y=0,Z=0),Scale3D=(X=1,Y=1,Z=1)))
    Primary=(Weight=1,Axis=(X=1,Y=0,Z=0),Target=(X=1,Y=0,Z=0),Kind=Location,Space="None")
    Secondary=(Weight=1,Axis=(X=0,Y=0,Z=0),Target=(X=1,Y=0,Z=0),Kind=Location)
    Weight=1
    bPropagateToChildren=True
    Primary.Target=&RigUnit_GetBoneTransform(Bone=clav_top_r).Transform.Translation
   )
   > &RigUnit_AimBone(
    Bone=clav_top_r
    DebugSettings=(Scale=10,WorldOffset=(Rotation=(X=0,Y=0,Z=0,W=1),Translation=(X=0,Y=0,Z=0),Scale3D=(X=1,Y=1,Z=1)))
    Primary=(Weight=1,Axis=(X=1,Y=0,Z=0),Target=(X=1,Y=0,Z=0),Kind=Location)
    Secondary=(Weight=1,Axis=(X=0,Y=0,Z=0),Target=(X=1,Y=0,Z=0),Kind=Location)
    Weight=1
    bPropagateToChildren=True
    Primary.Target=&RigUnit_GetBoneTransform(Bone=clav_bot_r).Transform.Translation
   )
   > &RigUnit_AimBone(
    Bone=shld_top_r
    DebugSettings=(bEnabled=false,Scale=10,WorldOffset=(Rotation=(X=0,Y=0,Z=0,W=1),Translation=(X=0,Y=0,Z=0),Scale3D=(X=1,Y=1,Z=1)))
    Primary=(Weight=1,Axis=(X=1,Y=0,Z=0),Target=(X=1,Y=0,Z=0),Kind=Location,Space="None")
    Secondary=(Weight=1,Axis=(X=0,Y=0,Z=0),Target=(X=1,Y=0,Z=0),Kind=Location)
    Weight=1
    bPropagateToChildren=True
    Primary.Target=&RigUnit_GetBoneTransform(Bone=shld_bot_r).Transform.Translation
   )
   > &RigUnit_AimBone(
    Bone=shld_bot_r
    DebugSettings=(Scale=10,WorldOffset=(Rotation=(X=0,Y=0,Z=0,W=1),Translation=(X=0,Y=0,Z=0),Scale3D=(X=1,Y=1,Z=1)))
    Primary=(Weight=1,Axis=(X=1,Y=0,Z=0),Target=(X=1,Y=0,Z=0),Kind=Location)
    Secondary=(Weight=1,Axis=(X=0,Y=0,Z=0),Target=(X=1,Y=0,Z=0),Kind=Location)
    Weight=1
    bPropagateToChildren=True
    Primary.Target=&RigUnit_GetBoneTransform(Bone=shld_top_r).Transform.Translation
   )
   > &RigUnit_AimBone(
    Bone=elbow_top_r
    DebugSettings=(bEnabled=false,Scale=10,WorldOffset=(Rotation=(X=0,Y=0,Z=0,W=1),Translation=(X=0,Y=0,Z=0),Scale3D=(X=1,Y=1,Z=1)))
    Primary=(Weight=1,Axis=(X=1,Y=0,Z=0),Target=(X=1,Y=0,Z=0),Kind=Location,Space="None")
    Secondary=(Weight=1,Axis=(X=0,Y=0,Z=0),Target=(X=1,Y=0,Z=0),Kind=Location)
    Weight=1
    bPropagateToChildren=True
    Primary.Target=&RigUnit_GetBoneTransform(Bone=elbow_bot_r).Transform.Translation
   )
   > &RigUnit_AimBone(
    Bone=elbow_bot_r
    DebugSettings=(Scale=10,WorldOffset=(Rotation=(X=0,Y=0,Z=0,W=1),Translation=(X=0,Y=0,Z=0),Scale3D=(X=1,Y=1,Z=1)))
    Primary=(Weight=1,Axis=(X=1,Y=0,Z=0),Target=(X=1,Y=0,Z=0),Kind=Location)
    Secondary=(Weight=1,Axis=(X=0,Y=0,Z=0),Target=(X=1,Y=0,Z=0),Kind=Location)
    Weight=1
    bPropagateToChildren=True
    Primary.Target=&RigUnit_GetBoneTransform(Bone=elbow_top_r).Transform.Translation
   )
   > &RigUnit_AimBone(
    Bone=bicep_bot_r
    DebugSettings=(bEnabled=false,Scale=10,WorldOffset=(Rotation=(X=0,Y=0,Z=0,W=1),Translation=(X=0,Y=0,Z=0),Scale3D=(X=1,Y=1,Z=1)))
    Primary=(Weight=1,Axis=(X=1,Y=0,Z=0),Target=(X=1,Y=0,Z=0),Kind=Location,Space="None")
    Secondary=(Weight=1,Axis=(X=0,Y=0,Z=0),Target=(X=1,Y=0,Z=0),Kind=Location)
    Weight=1
    bPropagateToChildren=True
    Primary.Target=&RigUnit_GetBoneTransform(Bone=bicep_top_r).Transform.Translation
   )
   > &RigUnit_AimBone(
    Bone=bicep_top_r
    DebugSettings=(Scale=10,WorldOffset=(Rotation=(X=0,Y=0,Z=0,W=1),Translation=(X=0,Y=0,Z=0),Scale3D=(X=1,Y=1,Z=1)))
    Primary=(Weight=1,Axis=(X=1,Y=0,Z=0),Target=(X=1,Y=0,Z=0),Kind=Location)
    Secondary=(Weight=1,Axis=(X=0,Y=0,Z=0),Target=(X=1,Y=0,Z=0),Kind=Location)
    Weight=1
    bPropagateToChildren=True
    Primary.Target=&RigUnit_GetBoneTransform(Bone=bicep_top_r).Transform.Translation
   )
   > &RigUnit_AimBone(
    Bone=spine_03_rod_l
    DebugSettings=(bEnabled=false,Scale=10,WorldOffset=(Rotation=(X=0,Y=0,Z=0,W=1),Translation=(X=0,Y=0,Z=0),Scale3D=(X=1,Y=1,Z=1)))
    Primary=(Weight=1,Axis=(X=1,Y=0,Z=0),Target=(X=1,Y=0,Z=0),Kind=Location,Space="None")
    Secondary=(Weight=1,Axis=(X=0,Y=0,Z=0),Target=(X=1,Y=0,Z=0),Kind=Location)
    Weight=1
    bPropagateToChildren=True
    Primary.Target=&RigUnit_GetBoneTransform(Bone=pelvis_rod_l).Transform.Translation
   )
   > &RigUnit_AimBone(
    Bone=pelvis_rod_l
    DebugSettings=(Scale=10,WorldOffset=(Rotation=(X=0,Y=0,Z=0,W=1),Translation=(X=0,Y=0,Z=0),Scale3D=(X=1,Y=1,Z=1)))
    Primary=(Weight=1,Axis=(X=1,Y=0,Z=0),Target=(X=1,Y=0,Z=0),Kind=Location)
    Secondary=(Weight=1,Axis=(X=0,Y=0,Z=0),Target=(X=1,Y=0,Z=0),Kind=Location)
    Weight=1
    bPropagateToChildren=True
    Primary.Target=&RigUnit_GetBoneTransform(Bone=spine_03_rod_l).Transform.Translation
   )
   > &RigUnit_AimBone(
    Bone=spine_03_rod_r
    DebugSettings=(bEnabled=false,Scale=10,WorldOffset=(Rotation=(X=0,Y=0,Z=0,W=1),Translation=(X=0,Y=0,Z=0),Scale3D=(X=1,Y=1,Z=1)))
    Primary=(Weight=1,Axis=(X=1,Y=0,Z=0),Target=(X=1,Y=0,Z=0),Kind=Location,Space="None")
    Secondary=(Weight=1,Axis=(X=0,Y=0,Z=0),Target=(X=1,Y=0,Z=0),Kind=Location)
    Weight=1
    bPropagateToChildren=True
    Primary.Target=&RigUnit_GetBoneTransform(Bone=pelvis_rod_r).Transform.Translation
   )
   > &RigUnit_AimBone(
    Bone=pelvis_rod_r
    DebugSettings=(Scale=10,WorldOffset=(Rotation=(X=0,Y=0,Z=0,W=1),Translation=(X=0,Y=0,Z=0),Scale3D=(X=1,Y=1,Z=1)))
    Primary=(Weight=1,Axis=(X=1,Y=0,Z=0),Target=(X=1,Y=0,Z=0),Kind=Location)
    Secondary=(Weight=1,Axis=(X=0,Y=0,Z=0),Target=(X=1,Y=0,Z=0),Kind=Location)
    Weight=1
    bPropagateToChildren=True
    Primary.Target=&RigUnit_GetBoneTransform(Bone=spine_03_rod_r).Transform.Translation
   )
   n1 = /Script/ControlRig.RigUnit_GetInitialBoneTransform(Bone=upperarm_l, Space=GlobalSpace) {is=struct RigVMUnitNode}
   n2 = &RigUnit_GetInitialBoneTransform(Bone=upperarm_alt_l, Space=GlobalSpace)
   /Script/ControlRig.RigUnit_SetBoneRotation(
    Bone=upperarm_alt_l
    ExecutePin=()
    Space=LocalSpace
    Weight=1
    bPropagateToChildren=True
    Rotation=&RigVMFunction_MathTransformMakeRelative(
     Global=&RigUnit_GetBoneTransform(Bone=upperarm_l, Space=LocalSpace).Transform
     Parent=&RigVMFunction_MathTransformMakeRelative(Global=@n1.Transform, Parent=@n2.Transform).Local
    ).Local.Rotation
   ) {is=struct RigVMUnitNode}
   n3 = &RigUnit_GetInitialBoneTransform(Bone=lowerarm_l, Space=GlobalSpace)
   n4 = &RigUnit_GetInitialBoneTransform(Bone=lowerarm_alt_l, Space=GlobalSpace)
   > &RigUnit_SetBoneRotation(
    Bone=lowerarm_alt_l
    Space=LocalSpace
    Weight=1
    bPropagateToChildren=True
    Rotation=&RigVMFunction_MathTransformMakeRelative(
     Global=&RigUnit_GetBoneTransform(Bone=lowerarm_l, Space=LocalSpace).Transform
     Parent=&RigVMFunction_MathTransformMakeRelative(Global=@n3.Transform, Parent=@n4.Transform).Local
    ).Local.Rotation
   )
   &RigUnit_AimBone(
    Bone=None
    DebugSettings=(Scale=10,WorldOffset=(Rotation=(X=0,Y=0,Z=0,W=1),Translation=(X=0,Y=0,Z=0),Scale3D=(X=1,Y=1,Z=1)))
    ExecutePin=()
    Primary=(Weight=1,Axis=(X=1,Y=0,Z=0),Target=(X=1,Y=0,Z=0),Kind=Location)
    Secondary=(Weight=1,Axis=(X=0,Y=0,Z=1),Target=(X=1,Y=0,Z=0),Kind=Location)
    Weight=1
    bPropagateToChildren=True
   )
   /Script/ControlRig.RigUnit_DebugTransformMutable(
    Color=(R=1,G=0,B=0,A=1)
    DebugDrawSettings=(DepthPriority=SDPG_Foreground,Lifetime=-1)
    ExecutePin=()
    Mode=Axes
    Scale=25
    Space=None
    Thickness=0.2
    WorldOffset=(Rotation=(X=0,Y=0,Z=0,W=1),Translation=(X=0,Y=0,Z=0),Scale3D=(X=1,Y=1,Z=1))
    bEnabled=false
    Transform=/Script/RigVM.RigVMFunction_MathTransformSelectBool(
     Condition=true
     IfTrue=&RigUnit_GetBoneTransform(Bone=thigh_attach_frt_01_l).Transform
     IfFalse=&RigUnit_GetBoneTransform(Bone=pelvis_attach_frt_02_l).Transform
    ) {is=struct RigVMUnitNode}.Result
   ) {is=struct RigVMUnitNode}
   &RigUnit_DebugTransformMutable(
    Color=(R=1,G=0,B=0,A=1)
    DebugDrawSettings=(DepthPriority=SDPG_Foreground,Lifetime=-1)
    ExecutePin=()
    Mode=Axes
    Scale=25
    Space=None
    Thickness=1
    WorldOffset=(Rotation=(X=0,Y=0,Z=0,W=1),Translation=(X=0,Y=0,Z=0),Scale3D=(X=1,Y=1,Z=1))
    bEnabled=false
    Transform=&RigVMFunction_MathTransformSelectBool(
     Condition=true
     IfTrue=&RigUnit_GetBoneTransform(Bone=thigh_attach_frt_01_l).Transform
     IfFalse=&RigUnit_GetBoneTransform(Bone=pelvis_attach_frt_02_l).Transform
    ).Result
   )
   &RigUnit_BeginExecution()
  ]}
]
hierarchy [
 root {loc=<0 0 0>, rot=<0 0 0 1>}
 pelvis {parent=root, loc=<0 -0.7071 93.1191>, rot=<0.0127 0.707 -0.0127 -0.707>}
 spine_01 {parent=pelvis, loc=<6.7494 -0.6388 0>, rot=<0 0 0.034 -0.9994>}
 spine_02 {parent=spine_01, loc=<7.2518 -0.035 0>, rot=<0 0 0.0238 -0.9997>}
 spine_03 {parent=spine_02, loc=<6.0104 0 0>, rot=<0 0 0.0149 -0.9999>}
 spine_04 {parent=spine_03, loc=<8.1024 0 0>, rot=<0 0 -0.0132 -0.9999>}
 spine_05 {parent=spine_04, loc=<10.0146 0 0>, rot=<0 0 -0.0015 -1>}
 clavicle_l {parent=spine_05, loc=<4.3575 -0.1466 -2.3223>, rot=<0.0465 -0.7289 -0.009 -0.6829>}
 upperarm_l {parent=clavicle_l, loc=<18.5074 0.1726 0.2023>, rot=<0.0182 -0.3281 -0.0261 -0.9441>}
 lowerarm_l {parent=upperarm_l, loc=<26.6514 0 0>, rot=<0.0391 -0.0099 -0.2453 -0.9686>}
 hand_l {parent=lowerarm_l, loc=<22.3491 0 0>, rot=<-0.7081 -0.0507 0.0975 -0.6975>}
 index_metacarpal_l {parent=hand_l, loc=<1.6535 0.0183 -2.96>, rot=<0.0158 -0.0124 -0.0173 -0.9996>}
 index_01_l {parent=index_metacarpal_l, loc=<8.3008 0.0358 -0.0524>, rot=<-0.0044 0.0485 0.0924 -0.9945>}
 index_02_l {parent=index_01_l, loc=<3.3587 -0.3003 0.0452>, rot=<-0.002 0.0027 0.3225 -0.9466>}
 index_03_l {parent=index_02_l, loc=<2.5578 -0.1231 -0.0252>, rot=<-0.002 0.0067 0.2558 -0.9667>}
 middle_metacarpal_l {parent=hand_l, loc=<2.1856 -0.1725 -0.5984>, rot=<0.006 0.0168 -0.0329 -0.9993>}
 middle_01_l {parent=middle_metacarpal_l, loc=<7.3275 -0.2448 0.0129>, rot=<-0.0208 0.0042 0.0858 -0.9961>}
 middle_02_l {parent=middle_01_l, loc=<4.0123 0.1401 0.1175>, rot=<-0.0218 0.0258 0.3512 -0.9357>}
 middle_03_l {parent=middle_02_l, loc=<2.6077 0.0534 0.1373>, rot=<0.0082 0.0188 0.1708 -0.9851>}
 pinky_metacarpal_l {parent=hand_l, loc=<1.7616 -0.9753 3.3814>, rot=<-0.0397 0.0814 -0.069 -0.9935>}
 pinky_01_l {parent=pinky_metacarpal_l, loc=<7.1709 -0.4831 -0.1094>, rot=<-0.0654 -0.0035 0.2242 -0.9723>}
 pinky_02_l {parent=pinky_01_l, loc=<2.8274 0.0893 -0.1043>, rot=<0.0037 0.0011 0.2854 -0.9584>}
 pinky_03_l {parent=pinky_02_l, loc=<2.2046 0.0307 -0.0078>, rot=<-0.0002 0.0098 0.1445 -0.9895>}
 ring_metacarpal_l {parent=hand_l, loc=<2.2074 -0.3625 1.3638>, rot=<-0.0284 0.0722 -0.0458 -0.9959>}
 ring_01_l {parent=ring_metacarpal_l, loc=<7.0295 -0.6197 -0.1446>, rot=<-0.0489 -0.0252 0.1892 -0.9804>}
 ring_02_l {parent=ring_01_l, loc=<3.6174 0.4647 -0.1027>, rot=<0 0 0.2806 -0.9598>}
 ring_03_l {parent=ring_02_l, loc=<2.4152 0.1141 0.0199>, rot=<0 0 0.1189 -0.9929>}
 thumb_01_l {parent=hand_l, loc=<1.3118 -0.3191 -2.459>, rot=<0.6546 -0.4042 0.0664 -0.6354>}
 thumb_02_l {parent=thumb_01_l, loc=<5.9336 -0.0448 0.0401>, rot=<-0.0037 0.0213 0.298 -0.9543>}
 thumb_03_l {parent=thumb_02_l, loc=<2.2223 0.3682 -0.0907>, rot=<0.0118 -0.013 0.0066 -0.9998>}
 lowerarm_twist_01_l {parent=lowerarm_l, loc=<17.8549 0 0>, rot=<0 0 0 1>}
 lowerarm_twist_02_l {parent=lowerarm_l, loc=<9.6199 0 0>, rot=<0 0 0 1>}
 elbow_01_l {parent=lowerarm_l, loc=<0 0 0>, rot=<0 0 0 1>}
 elbow_02_l {parent=elbow_01_l, loc=<-2.7833 -3.7289 -0.6399>, rot=<-0.1494 -0.7257 0.6585 0.1318>}
 elbow_bot_l {parent=elbow_02_l, loc=<0 0 0>, rot=<0 0 0 -1>}
 bicep_bot_l {parent=lowerarm_l, loc=<2.232 2.492 0.4278>, rot=<-0.2156 -0.7072 0.6431 0.1993>}
 elbow_top_l {parent=upperarm_l, loc=<7.7513 -6.3771 -0.228>, rot=<0.6998 0.0348 -0.0381 -0.7125>}
 upperarm_twist_01_l {parent=upperarm_l, loc=<0.0041 0 0>, rot=<0 0 0 -1>}
 bicep_top_l {parent=upperarm_l, loc=<8.3384 5.1921 0.0977>, rot=<0.7016 -0.0338 0.0368 -0.7108>}
 clav_bot_l {parent=clavicle_l, loc=<12.5675 -0.039 6.578>, rot=<-0.1891 -0.6031 0.7629 0.1356>}
 chest_top_l {parent=clavicle_l, loc=<12.6149 6.1374 0.077>, rot=<-0.3227 -0.0484 0.94 0.0994>}
 armpit_top_l {parent=clavicle_l, loc=<13.2934 -0.0362 -7.0314>, rot=<-0.4958 -0.0434 0.8667 0.0339>}
 dyn_parent_cable_01_l {parent=clavicle_l, loc=<6.6818 0 0>, rot=<-0.1024 0.5285 -0.3604 -0.7618>}
 dyn_cable_01_l {parent=dyn_parent_cable_01_l, loc=<0 0 0>, rot=<0 0 0 -1>}
 NewControl_0Space {parent=clavicle_l, type=Null, loc=<0 0 0>, rot=<0 0 0 1>}
 clavicle_l_ctrl {parent=NewControl_0Space, type=Control, loc=<0 0 0>, rot=<0 0 0 1>}
 clavicle_r {parent=spine_05, loc=<4.3575 -0.1466 2.3223>, rot=<0.7289 0.0465 -0.6829 0.009>}
 upperarm_r {parent=clavicle_r, loc=<-18.5074 -0.1726 -0.2023>, rot=<0.0182 -0.3281 -0.0261 -0.9441>}
 lowerarm_r {parent=upperarm_r, loc=<-26.6515 0 0>, rot=<0.0391 -0.0099 -0.2453 -0.9686>}
 hand_r {parent=lowerarm_r, loc=<-22.3492 0 0>, rot=<-0.7081 -0.0507 0.0975 -0.6975>}
 index_metacarpal_r {parent=hand_r, loc=<-1.6534 -0.0181 2.9601>, rot=<0.0158 -0.0124 -0.0173 -0.9996>}
 index_01_r {parent=index_metacarpal_r, loc=<-8.3008 -0.0357 0.0524>, rot=<-0.0044 0.0485 0.0924 -0.9945>}
 index_02_r {parent=index_01_r, loc=<-3.3589 0.3003 -0.0453>, rot=<-0.002 0.0027 0.3225 -0.9466>}
 index_03_r {parent=index_02_r, loc=<-2.5578 0.1231 0.0251>, rot=<-0.002 0.0067 0.2558 -0.9667>}
 middle_metacarpal_r {parent=hand_r, loc=<-2.1856 0.1726 0.5984>, rot=<0.006 0.0168 -0.0329 -0.9993>}
 middle_01_r {parent=middle_metacarpal_r, loc=<-7.3273 0.2447 -0.0129>, rot=<-0.0208 0.0042 0.0858 -0.9961>}
 middle_02_r {parent=middle_01_r, loc=<-4.0142 -0.1411 -0.07>, rot=<-0.0217 0.0259 0.3512 -0.9357>}
 middle_03_r {parent=middle_02_r, loc=<-2.6074 -0.0534 -0.1312>, rot=<0.0083 0.0188 0.1708 -0.9851>}
 pinky_metacarpal_r {parent=hand_r, loc=<-1.7617 0.9753 -3.3814>, rot=<-0.0397 0.0814 -0.069 -0.9935>}
 pinky_01_r {parent=pinky_metacarpal_r, loc=<-7.1707 0.4832 0.1094>, rot=<-0.0654 -0.0035 0.2242 -0.9723>}
 pinky_02_r {parent=pinky_01_r, loc=<-2.8274 -0.0892 0.1043>, rot=<0.0037 0.0011 0.2854 -0.9584>}
 ring_metacarpal_r {parent=hand_r, loc=<-2.2073 0.3627 -1.3638>, rot=<-0.0284 0.0722 -0.0458 -0.9959>}
 ring_01_r {parent=ring_metacarpal_r, loc=<-7.0297 0.6197 0.1446>, rot=<-0.0489 -0.0252 0.1892 -0.9804>}
 ring_02_r {parent=ring_01_r, loc=<-3.6171 -0.4648 0.1027>, rot=<0 0 0.2806 -0.9598>}
 ring_03_r {parent=ring_02_r, loc=<-2.4153 -0.1141 -0.0199>, rot=<0 0 0.1189 -0.9929>}
 thumb_01_r {parent=hand_r, loc=<-1.3119 0.3193 2.459>, rot=<0.6546 -0.4042 0.0664 -0.6354>}
 thumb_02_r {parent=thumb_01_r, loc=<-5.9335 0.0447 -0.0401>, rot=<-0.0037 0.0213 0.298 -0.9543>}
 thumb_03_r {parent=thumb_02_r, loc=<-2.2223 -0.3682 0.0907>, rot=<0.0118 -0.013 0.0066 -0.9998>}
 lowerarm_twist_01_r {parent=lowerarm_r, loc=<-17.855 0 0>, rot=<0 0 0 -1>}
 lowerarm_twist_02_r {parent=lowerarm_r, loc=<-9.6199 0 0>, rot=<0 0 0 -1>}
 elbow_01_r {parent=lowerarm_r, loc=<0.0002 -0.0001 -0.0001>, rot=<0 0 0 -1>}
 elbow_02_r {parent=elbow_01_r, loc=<2.7833 3.7289 0.6398>, rot=<-0.7257 0.1494 0.1318 -0.6585>}
 elbow_bot_r {parent=elbow_02_r, loc=<0 0 0>, rot=<0 0 0 1>}
 bicep_bot_r {parent=lowerarm_r, loc=<-2.2319 -2.4921 -0.4279>, rot=<-0.7072 0.2156 0.1993 -0.6431>}
 elbow_top_r {parent=upperarm_r, loc=<-7.7512 6.3771 0.228>, rot=<-0.0348 0.6998 0.7125 -0.0381>}
 bicep_top_r {parent=upperarm_r, loc=<-8.3382 -5.1921 -0.0979>, rot=<0.0338 0.7016 0.7108 0.0368>}
 armpit_top_r {parent=clavicle_r, loc=<-13.2934 0.0362 7.0318>, rot=<-0.0434 0.4958 0.0339 -0.8667>}
 chest_top_r {parent=clavicle_r, loc=<-12.615 -6.1374 -0.0767>, rot=<-0.0484 0.3227 0.0994 -0.94>}
 clav_bot_r {parent=clavicle_r, loc=<-12.5674 0.039 -6.5781>, rot=<-0.6031 0.1891 0.1356 -0.7629>}
 dyn_parent_cable_01_r {parent=clavicle_r, loc=<-6.6819 0 0.0005>, rot=<-0.5285 -0.1024 0.7618 -0.3604>}
 dyn_cable_01_r {parent=dyn_parent_cable_01_r, loc=<0 0 0>, rot=<0 0 0 -1>}
 neck_01 {parent=spine_05, loc=<11.3138 0.1529 0>, rot=<0 0 -0.2003 -0.9797>}
 neck_02 {parent=neck_01, loc=<4.5627 0 0>, rot=<0 0 0.0504 -0.9987>}
 head {parent=neck_02, loc=<5.3914 0 0>, rot=<0 0 0.115 -0.9934>}
 dyn_parent_head_rod_l {parent=head, loc=<2.4017 -1.6086 -4.071>, rot=<-0.0602 0.6866 0.7244 -0.0178>}
 head_rod_l {parent=dyn_parent_head_rod_l, loc=<0 0 0>, rot=<0 0 0 1>}
 dyn_parent_head_rod_r {parent=head, loc=<2.4022 -1.6086 4.071>, rot=<0.0602 -0.6866 0.7244 -0.0178>}
 head_rod_r {parent=dyn_parent_head_rod_r, loc=<0 0 0>, rot=<0 0 0 -1>}
 dyn_parent_head_wire_l {parent=head, loc=<3.523 0.9039 -5.2942>, rot=<0.0782 -0.7499 -0.6449 0.125>}
 head_wire_l {parent=dyn_parent_head_wire_l, loc=<0 0 0>, rot=<0 0 0 -1>}
 dyn_parent_head_wire_r {parent=head, loc=<3.5234 0.9039 5.2942>, rot=<0.0782 -0.7499 0.6449 -0.125>}
 head_wire_r {parent=dyn_parent_head_wire_r, loc=<0 0 0>, rot=<0 0 0 -1>}
 clav_top_l {parent=spine_05, loc=<8.6769 6.2204 -3.0779>, rot=<0.6895 -0.5146 -0.4286 -0.276>}
 scapula_l {parent=spine_05, loc=<3.1546 -9.8271 -10.0524>, rot=<0.0284 -0.7065 -0.0284 -0.7065>}
 scapula_tip_l {parent=scapula_l, loc=<5.7465 3.7124 7.5693>, rot=<-0.0619 -0.9707 0.0148 0.2315>}
 shld_bot_l {parent=scapula_tip_l, loc=<0 -1.0931 0>, rot=<0 0 0 1>}
 a_clavicle_l {parent=spine_05, loc=<4.3575 -0.1466 -2.3223>, rot=<0.0465 -0.7289 -0.009 -0.6829>}
 a_upperarm_l {parent=a_clavicle_l, loc=<18.5074 0.1726 0.2023>, rot=<0.0182 -0.3281 -0.0261 -0.9441>}
 a_lowerarm_l {parent=a_upperarm_l, loc=<26.6514 0 0>, rot=<0.0391 -0.0099 -0.2453 -0.9686>}
 a_hand_l {parent=a_lowerarm_l, loc=<22.3491 0 0>, rot=<-0.7081 -0.0507 0.0975 -0.6975>}
 scapula_r {parent=spine_05, loc=<3.1546 -9.8271 10.0524>, rot=<0.7065 0.0284 -0.7065 0.0284>}
 scapula_tip_r {parent=scapula_r, loc=<-5.7465 -3.7124 -7.569>, rot=<0.0148 -0.2315 0.0619 -0.9707>}
 shld_bot_r {parent=scapula_tip_r, loc=<0.0002 1.0932 -0.0003>, rot=<0 0 0 1>}
 clav_top_r {parent=spine_05, loc=<8.6764 6.2204 3.0779>, rot=<0.6895 -0.5146 0.4286 0.276>}
 dyn_parent_neck_rod_l {parent=spine_05, loc=<9.9561 -0.2988 -4.9148>, rot=<-0.7232 -0.0439 -0.0848 -0.684>}
 neck_rod_l {parent=dyn_parent_neck_rod_l, loc=<0 0 0>, rot=<0 0 0 1>}
 dyn_parent_neck_rod_r {parent=spine_05, loc=<9.956 -0.2988 4.9148>, rot=<0.7232 0.0439 -0.0848 -0.684>}
 neck_rod_r {parent=dyn_parent_neck_rod_r, loc=<0 0 0>, rot=<0 0 0 -1>}
 dyn_parent_neck_wire_l {parent=spine_05, loc=<9.5436 4.1153 -3.2177>, rot=<-0.7057 -0.0176 0.0616 -0.7056>}
 neck_wire_l {parent=dyn_parent_neck_wire_l, loc=<0 0 0>, rot=<0 0 0 1>}
 dyn_parent_neck_wire_r {parent=spine_05, loc=<9.5439 4.1153 3.2177>, rot=<0.7057 0.0176 0.0616 -0.7056>}
 neck_wire_r {parent=dyn_parent_neck_wire_r, loc=<0 0 0>, rot=<0 0 0 1>}
 dyn_shld_top_parent_l {parent=spine_05, loc=<17.3649 -4.0103 -1.469>, rot=<0.0284 -0.7065 -0.0284 -0.7065>}
 shld_top_l {parent=dyn_shld_top_parent_l, loc=<0.5338 -0.0697 -0.0842>, rot=<0.0148 -0.2315 0.0619 -0.9707>}
 dyn_shld_top_parent_r {parent=spine_05, loc=<17.3652 -4.0103 1.469>, rot=<0.0284 -0.7065 -0.0284 -0.7065>}
 shld_top_r {parent=dyn_shld_top_parent_r, loc=<-0.5338 -0.0697 -0.084>, rot=<-0.2315 0.0148 0.9707 -0.0619>}
 chest_bot_l {parent=spine_04, loc=<4.3214 8.6098 -2.7022>, rot=<0.8805 -0.0784 -0.4605 0.0809>}
 chest_bot_r {parent=spine_04, loc=<4.3213 8.6098 2.7022>, rot=<0.8805 -0.0784 0.4605 -0.0809>}
 dyn_torso_cable_01_l {parent=spine_04, loc=<2.9823 1.0564 -5.5187>, rot=<0.0186 -0.4478 -0.0372 -0.8932>}
 dyn_torso_cable_01_r {parent=spine_04, loc=<2.9825 1.0564 5.5187>, rot=<-0.0186 0.4478 -0.0372 -0.8932>}
 armpit_bot_l {parent=spine_03, loc=<6.9372 -0.9266 -9.8989>, rot=<0.9559 -0.01 -0.2915 0.0351>}
 armpit_bot_r {parent=spine_03, loc=<6.9376 -0.9265 9.8989>, rot=<0.9559 -0.01 0.2915 -0.0351>}
 spine_03_rod_l {parent=spine_03, loc=<1.0809 0.8366 -9.7204>, rot=<0.0444 -0.0023 0.999 0.0079>}
 spine_03_rod_r {parent=spine_03, loc=<1.0808 0.8366 9.7204>, rot=<-0.0444 0.0023 0.999 0.0079>}
 thigh_l {parent=pelvis, loc=<-2.5693 -0.0109 -7.7501>, rot=<0.0012 0.0572 0.0155 -0.9982>}
 calf_l {parent=thigh_l, loc=<-39.0484 0 0>, rot=<0.0006 -0.0161 -0.0391 -0.9991>}
 calf_twist_01_l {parent=calf_l, loc=<-35.8131 -0.0967 0.0397>, rot=<0.0237 0.0095 -0.0035 -0.9997>}
 foot_l {parent=calf_l, loc=<-42.561 0.1077 0.4292>, rot=<0.0418 -0.0455 0.0343 -0.9975>}
 ball_l {parent=foot_l, loc=<-6.1599 13.8264 0.6806>, rot=<0.0001 -0.0001 -0.7063 -0.7079>}
 thigh_attach_frt_01_l {parent=thigh_l, loc=<-29.4094 2.1539 -0.334>, rot=<0.9913 0.0768 0.1061 0.0143>}
 thigh_attach_bck_01_l {parent=thigh_l, loc=<-34.0638 -0.7569 -0.6144>, rot=<0.9924 -0.0891 0.0791 -0.0294>}
 NewControlSpace {parent=thigh_l, type=Null, loc=<0 0 0>, rot=<0 0 0 1>}
 NewControl {parent=NewControlSpace, type=Control, loc=<0 0 0>, rot=<0 0 0 1>}
 thigh_r {parent=pelvis, loc=<-2.5691 -0.0109 7.7501>, rot=<0.0572 -0.0012 0.9982 0.0155>}
 calf_r {parent=thigh_r, loc=<39.0487 0 0>, rot=<0.0006 -0.0161 -0.0391 -0.9991>}
 calf_twist_01_r {parent=calf_r, loc=<35.8132 0.0967 -0.0397>, rot=<0.0237 0.0095 -0.0035 -0.9997>}
 foot_r {parent=calf_r, loc=<42.5611 -0.1077 -0.4292>, rot=<0.0418 -0.0455 0.0343 -0.9975>}
 ball_r {parent=foot_r, loc=<6.1598 -13.8264 -0.6806>, rot=<0.0001 -0.0001 -0.7063 -0.7079>}
 thigh_attach_frt_01_r {parent=thigh_r, loc=<29.4096 -2.1539 0.334>, rot=<0.0768 -0.9913 0.0143 -0.1061>}
 thigh_attach_bck_01_r {parent=thigh_r, loc=<34.064 0.7569 0.6144>, rot=<-0.0891 -0.9924 -0.0294 -0.0791>}
 pelvis_attach_frt_01_l {parent=pelvis, loc=<-6.1929 6.0857 -3.5506>, rot=<0 0 0.9998 0.0179>}
 pelvis_attach_frt_02_l {parent=pelvis_attach_frt_01_l, loc=<2.2562 0 -0.0618>, rot=<-0.0186 -0.1626 -0.0783 -0.9834>}
 pelvis_attach_bck_01_l {parent=pelvis, loc=<-7.245 -6.3314 -4.5163>, rot=<0 0 0.9998 0.0179>}
 pelvis_attach_bck_02_l {parent=pelvis_attach_bck_01_l, loc=<1.6559 -0.4063 -0.096>, rot=<0.0344 -0.1357 0.0849 -0.9865>}
 pelvis_attach_frt_01_r {parent=pelvis, loc=<-6.1929 6.0857 3.5506>, rot=<0 0 0.9998 0.0179>}
 pelvis_attach_frt_02_r {parent=pelvis_attach_frt_01_r, loc=<2.2562 0 0.0619>, rot=<0.0186 0.1626 -0.0783 -0.9834>}
 pelvis_attach_bck_01_r {parent=pelvis, loc=<-7.245 -6.3314 4.5163>, rot=<0 0 0.9998 0.0179>}
 pelvis_attach_bck_02_r {parent=pelvis_attach_bck_01_r, loc=<1.6559 -0.4063 0.096>, rot=<-0.0344 0.1357 0.0849 -0.9865>}
 pelvis_rod_l {parent=pelvis, loc=<7.4322 1.0366 -8.483>, rot=<0.9958 -0.0804 -0.0441 -0.0055>}
 pelvis_rod_r {parent=pelvis, loc=<7.4318 1.0366 8.483>, rot=<0.9958 -0.0804 0.0441 0.0055>}
]
