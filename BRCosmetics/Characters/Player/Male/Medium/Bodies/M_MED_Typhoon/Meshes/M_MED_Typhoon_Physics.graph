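# Physics asset for the M_MED_Typhoon male medium player body cosmetic, covering its left and right dangling cable bones rather than a full-body ragdoll.
# Small capsules on each bone form chains rooted at spine_04.
# Each cable hangs with limited swing and tight twist from a tiny parent body, whose angular motion to the spine is free.
# Torso cables attach directly to the spine with narrower limits, slightly asymmetric between sides.
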PhysicsAsset /BRCosmetics/Characters/Player/Male/Medium/Bodies/M_MED_Typhoon/Meshes/M_MED_Typhoon_Physics
bodies {
 dyn_cable_01_l [{shapeType=Capsule, length=5.75, radius=1.75, center=<0.25 0 0>, rotation=<rot 90 0 0>}]
 dyn_cable_01_r [{shapeType=Capsule, length=5.75, radius=1.75, center=<0.25 0 0>, rotation=<rot 90 0 0>}]
 dyn_parent_cable_01_l [{shapeType=Capsule, length=0.505, radius=0.505, rotation=<rot 0 -90 0>}]
 dyn_parent_cable_01_r [{shapeType=Capsule, length=0.505, radius=0.505, rotation=<rot 0 -90 0>}]
 dyn_torso_cable_01_l [{shapeType=Capsule, length=4.25, radius=4.25, center=<0.45 0 0>, rotation=<rot 90 0 0>}]
 dyn_torso_cable_01_r [{shapeType=Capsule, length=3.75, radius=3.25, center=<0.5 0 0>, rotation=<rot 90 0 0>}]
 spine_04 [{shapeType=Capsule, length=1, radius=1}]
}
constraints [
 {bone1Name=dyn_cable_01_l, bone2Name=dyn_parent_cable_01_l, swing1LimitDegrees=30, swing2LimitDegrees=30, twistLimitDegrees=2}
 {bone1Name=dyn_cable_01_r, bone2Name=dyn_parent_cable_01_r, swing1LimitDegrees=30, swing2LimitDegrees=30, twistLimitDegrees=2}
 {
  bone1Name=dyn_parent_cable_01_l
  bone2Name=spine_04
  swing1LimitDegrees=45
  swing1Motion=Free
  swing2LimitDegrees=45
  swing2Motion=Free
  twistLimitDegrees=45
  twistMotion=Free
 }
 {
  bone1Name=dyn_parent_cable_01_r
  bone2Name=spine_04
  swing1LimitDegrees=45
  swing1Motion=Free
  swing2LimitDegrees=45
  swing2Motion=Free
  twistLimitDegrees=45
  twistMotion=Free
 }
 {bone1Name=dyn_torso_cable_01_l, bone2Name=spine_04, swing1LimitDegrees=20, swing2LimitDegrees=20, twistLimitDegrees=3}
 {bone1Name=dyn_torso_cable_01_r, bone2Name=spine_04, swing1LimitDegrees=15, swing2LimitDegrees=15, twistLimitDegrees=3}
]
