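# Post-process style Anim Blueprint for the Typhoon medium male player body, adding secondary motion to the pose copied from the attached parent mesh.
# Head twist relative to spine_05 sets an interpolated alpha that lowers both dynamic shoulder-top bones.
# Flap nodes counter-rotate pelvis attachments from thigh rotation with clamping, and weighted constraints place scapula tips.
# A low-gravity RigidBody simulation then feeds a body-specific Control Rig before output.
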AnimBlueprint /BRCosmetics/Characters/Player/Male/Medium/Bodies/M_MED_Typhoon/Meshes/M_MED_Typhoon_AnimBP
target_skeleton=/Game/Characters/Player/Male/Medium/Bodies/M_MED_Typhoon/Meshes/M_MED_Typhoon_Skeleton
graphs [
 AnimGraph {nodes [
   n1 = Get head_back() -> (head_back:double)
   n2 = "Transform (Modify) Bone - Bone: dyn_shld_top_parent_l"(
    ComponentPose=Local To Component(LocalPose=Copy Pose From Mesh() {bUseAttachedParent=true})
    Alpha=@n1
    Translation="0,0,-2"
    Rotation="0,0,0"
    Scale="1,1,1"
   ) {
    translationMode=BMM_Additive
    translationSpace=BCS_BoneSpace
    alphaBoolBlend {blendOption=Linear}
    alphaScaleBiasClamp {bInterpResult=true, interpSpeedDecreasing=20, interpSpeedIncreasing=20}
    boneToModify {boneName=dyn_shld_top_parent_l}
    translation {z=-2}
   }
   n3 = "Transform (Modify) Bone - Bone: dyn_shld_top_parent_r"(ComponentPose=@n2, Alpha=@n1, Translation="0,0,-2", Rotation="0,0,0", Scale="1,1,1") {
    translationMode=BMM_Additive
    translationSpace=BCS_BoneSpace
    alphaBoolBlend {blendOption=Linear}
    alphaScaleBiasClamp {bInterpResult=true, interpSpeedDecreasing=20, interpSpeedIncreasing=20}
    boneToModify {boneName=dyn_shld_top_parent_r}
    translation {z=-2}
   }
   n4 = Flap Node - Bone: pelvis_attach_frt_01_l!(ComponentPose:Component Space Pose Link Structure=@n3, Alpha:float=1) -> (Pose:Component Space Pose Link Structure) {
    is=title FortAnimGraphNode_Flap
    bUseClamp=true
    targetClampMax=10
    targetClampMin=-10
    targetComponent=RotationZ
    alphaBoolBlend {blendOption=Linear}
    sourceDrives [{
      bUseRange=false
      drivingCurve=null
      multiplier=-1
      rangeMax=1
      rangeMin=-1
      remappedMax=1
      remappedMin=0
      sourceComponent=RotationZ
      useQuaternion=true
      sourceBone {boneName=thigh_l}
     }]
    targetBone {boneName=pelvis_attach_frt_01_l}
   }
   n5 = Flap Node - Bone: pelvis_attach_bck_01_l!(ComponentPose:Component Space Pose Link Structure=@n4, Alpha:float=1) -> (Pose:Component Space Pose Link Structure) {
    is=title FortAnimGraphNode_Flap
    bUseClamp=true
    targetClampMax=10
    targetClampMin=-10
    targetComponent=RotationZ
    alphaBoolBlend {blendOption=Linear}
    sourceDrives [{
      bUseRange=false
      drivingCurve=null
      multiplier=-1
      rangeMax=1
      rangeMin=-1
      remappedMax=1
      remappedMin=0
      sourceComponent=RotationZ
      useQuaternion=true
      sourceBone {boneName=thigh_l}
     }]
    targetBone {boneName=pelvis_attach_bck_01_l}
   }
   n6 = Flap Node - Bone: pelvis_attach_frt_01_r!(ComponentPose:Component Space Pose Link Structure=@n5, Alpha:float=1) -> (Pose:Component Space Pose Link Structure) {
    is=title FortAnimGraphNode_Flap
    bUseClamp=true
    targetClampMax=10
    targetClampMin=-10
    targetComponent=RotationZ
    alphaBoolBlend {blendOption=Linear}
    sourceDrives [{
      bUseRange=false
      drivingCurve=null
      multiplier=-1
      rangeMax=1
      rangeMin=-1
      remappedMax=1
      remappedMin=0
      sourceComponent=RotationZ
      useQuaternion=true
      sourceBone {boneName=thigh_r}
     }]
    targetBone {boneName=pelvis_attach_frt_01_r}
   }
   n7 = Flap Node - Bone: pelvis_attach_bck_01_r!(ComponentPose:Component Space Pose Link Structure=@n6, Alpha:float=1) -> (Pose:Component Space Pose Link Structure) {
    is=title FortAnimGraphNode_Flap
    bUseClamp=true
    targetClampMax=10
    targetClampMin=-10
    targetComponent=RotationZ
    alphaBoolBlend {blendOption=Linear}
    sourceDrives [{
      bUseRange=false
      drivingCurve=null
      multiplier=-1
      rangeMax=1
      rangeMin=-1
      remappedMax=1
      remappedMin=0
      sourceComponent=RotationZ
      useQuaternion=true
      sourceBone {boneName=thigh_r}
     }]
    targetBone {boneName=pelvis_attach_bck_01_r}
   }
   n8 = "Constraint - scapula_tip_l (PPPPP)"!(
    ComponentPose=@n7
    Alpha=1
    ConstraintWeights_0=0.0162
    ConstraintWeights_1:float=0.1106
    ConstraintWeights_2:float=0.0722
    ConstraintWeights_3:float=0.328
    ConstraintWeights_4:float=0.4731
   ) -> (Pose) {
    alphaBoolBlend {blendOption=Linear}
    boneToModify {boneName=scapula_tip_l}
    constraintSetup [
     {targetBone {boneName=neck_02}}
     {targetBone {boneName=spine_03}}
     {targetBone {boneName=spine_04}}
     {targetBone {boneName=spine_05}}
     {targetBone {boneName=clavicle_l}}
    ]
    constraintWeights [1,1,1,1,1]
   }
   n9 = "Constraint - scapula_tip_r (PPPPP)"!(
    ComponentPose=@n8
    Alpha=1
    ConstraintWeights_0=0.4732
    ConstraintWeights_1:float=0.0161
    ConstraintWeights_2:float=0.1106
    ConstraintWeights_3:float=0.0722
    ConstraintWeights_4:float=0.3279
   ) -> (Pose) {
    alphaBoolBlend {blendOption=Linear}
    boneToModify {boneName=scapula_tip_r}
    constraintSetup [
     {targetBone {boneName=clavicle_r}}
     {targetBone {boneName=neck_02}}
     {targetBone {boneName=spine_03}}
     {targetBone {boneName=spine_04}}
     {targetBone {boneName=spine_05}}
    ]
    constraintWeights [1,1,1,1,1]
   }
   n10 = RigidBody(ComponentPose=@n9) {
    bOverrideWorldGravity=true
    overridePhysicsAsset=&/M_MED_Typhoon_Physics
    simulationSpace=BaseBoneSpace
    alphaBoolBlend {blendOption=Linear}
    baseBoneRef {boneName=root}
    componentAppliedLinearAccClamp=<1000 1000 1000>
    componentLinearAccScale=<0.5 0.5 0.5>
    componentLinearVelScale=<0.5 0.5 0.5>
    overrideWorldGravity {z=-90}
   }
   n11 = Control Rig(Source=Component To Local(ComponentPose=@n10)) {
    alphaBoolBlend {blendOption=Linear}
    controlRigAssetReference {blueprintRigClass=&/M_MED_Typhoon_CtrlRig.M_MED_Typhoon_CtrlRig_C}
   }
   Output Pose(Result=@n11) {layerGroup=None}
  ]}
 EventGraph {nodes [
   Event BlueprintInitializeAnimation()
   > Set Body_Mesh(Body_Mesh=GetOwnerSkeletalMeshComponent())
   Event BlueprintUpdateAnimation()
   > Get!(execute:Exec) -> (then:Exec, else:Exec, Body_Mesh:Skeletal Mesh Component Object Reference)
   > Sequence()
   > Set head_back(head_back:double=GetAlphaFromReferenceTransformDelta(
     SkeletalMeshComponent=Get Body_Mesh()
     BoneName=head
     BaseName=spine_05
     DeltaComponentType=QuaternionTwist
     InputRange=(X=0,Y=48)
     TwistAxis="0,0,1"
    )) -> (Output_Get:double)
  ]}
]
variables [
 Body_Mesh {type=SkeletalMeshComponent}
 BodyMesh {type=SkeletalMeshComponent}
 head_back {type=number}
 Head_Mesh {type=SkeletalMeshComponent}
 neck_alpha {type=number}
 Pawn Owner {type=PlayerPawn_Athena_C}
 PawnOwner {type=PlayerPawn_Athena_C}
]
