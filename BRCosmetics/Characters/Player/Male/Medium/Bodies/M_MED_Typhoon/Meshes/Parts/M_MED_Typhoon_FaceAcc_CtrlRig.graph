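# Control Rig for the face accessory of the Typhoon medium male cosmetic body, keeping its left and right rods and wires lined up between head and neck.
# A single chain of eight AimBone units points each head rod or wire bone at its neck counterpart, then the neck bone back at the head one.
# Each aim reads the partner bone's transform translation as a location target on the X axis, with no secondary axis, and propagates to children.
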ControlRigBlueprint /BRCosmetics/Characters/Player/Male/Medium/Bodies/M_MED_Typhoon/Meshes/Parts/M_MED_Typhoon_FaceAcc_CtrlRig
graphs [
 RigVMFunctionLibrary {links [], nodes []}
 RigVMModel {nodes [
   RigUnit_BeginExecution = &RigUnit_BeginExecution()
   /Script/ControlRig.RigUnit_AimBone(
    Bone=head_rod_l
    DebugSettings=(bEnabled=false,Scale=10,WorldOffset=(Rotation=(X=0,Y=0,Z=0,W=1),Translation=(X=0,Y=0,Z=0),Scale3D=(X=1,Y=1,Z=1)))
    Primary=(Weight=1,Axis=(X=1,Y=0,Z=0),Target=(X=1,Y=0,Z=0),Kind=Location,Space="None")
    Secondary=(Weight=1,Axis=(X=0,Y=0,Z=0),Target=(X=1,Y=0,Z=0),Kind=Location)
    Weight=1
    bPropagateToChildren=True
    ExecutePin=@RigUnit_BeginExecution.ExecutePin
    Primary.Target=&RigUnit_GetBoneTransform(Bone=neck_rod_l).Transform.Translation
   ) {is=struct RigVMUnitNode}
   > &RigUnit_AimBone(
    Bone=neck_rod_l
    DebugSettings=(Scale=10,WorldOffset=(Rotation=(X=0,Y=0,Z=0,W=1),Translation=(X=0,Y=0,Z=0),Scale3D=(X=1,Y=1,Z=1)))
    Primary=(Weight=1,Axis=(X=1,Y=0,Z=0),Target=(X=1,Y=0,Z=0),Kind=Location)
    Secondary=(Weight=1,Axis=(X=0,Y=0,Z=0),Target=(X=1,Y=0,Z=0),Kind=Location)
    Weight=1
    bPropagateToChildren=True
    Primary.Target=&RigUnit_GetBoneTransform(Bone=head_rod_l).Transform.Translation
   )
   > &RigUnit_AimBone(
    Bone=head_rod_r
    DebugSettings=(bEnabled=false,Scale=10,WorldOffset=(Rotation=(X=0,Y=0,Z=0,W=1),Translation=(X=0,Y=0,Z=0),Scale3D=(X=1,Y=1,Z=1)))
    Primary=(Weight=1,Axis=(X=1,Y=0,Z=0),Target=(X=1,Y=0,Z=0),Kind=Location,Space="None")
    Secondary=(Weight=1,Axis=(X=0,Y=0,Z=0),Target=(X=1,Y=0,Z=0),Kind=Location)
    Weight=1
    bPropagateToChildren=True
    Primary.Target=&RigUnit_GetBoneTransform(Bone=neck_rod_r).Transform.Translation
   )
   > &RigUnit_AimBone(
    Bone=neck_rod_r
    DebugSettings=(Scale=10,WorldOffset=(Rotation=(X=0,Y=0,Z=0,W=1),Translation=(X=0,Y=0,Z=0),Scale3D=(X=1,Y=1,Z=1)))
    Primary=(Weight=1,Axis=(X=1,Y=0,Z=0),Target=(X=1,Y=0,Z=0),Kind=Location)
    Secondary=(Weight=1,Axis=(X=0,Y=0,Z=0),Target=(X=1,Y=0,Z=0),Kind=Location)
    Weight=1
    bPropagateToChildren=True
    Primary.Target=&RigUnit_GetBoneTransform(Bone=head_rod_r).Transform.Translation
   )
   > &RigUnit_AimBone(
    Bone=head_wire_l
    DebugSettings=(bEnabled=false,Scale=10,WorldOffset=(Rotation=(X=0,Y=0,Z=0,W=1),Translation=(X=0,Y=0,Z=0),Scale3D=(X=1,Y=1,Z=1)))
    Primary=(Weight=1,Axis=(X=1,Y=0,Z=0),Target=(X=1,Y=0,Z=0),Kind=Location,Space="None")
    Secondary=(Weight=1,Axis=(X=0,Y=0,Z=0),Target=(X=1,Y=0,Z=0),Kind=Location)
    Weight=1
    bPropagateToChildren=True
    Primary.Target=&RigUnit_GetBoneTransform(Bone=neck_wire_l).Transform.Translation
   )
   > &RigUnit_AimBone(
    Bone=neck_wire_l
    DebugSettings=(Scale=10,WorldOffset=(Rotation=(X=0,Y=0,Z=0,W=1),Translation=(X=0,Y=0,Z=0),Scale3D=(X=1,Y=1,Z=1)))
    Primary=(Weight=1,Axis=(X=1,Y=0,Z=0),Target=(X=1,Y=0,Z=0),Kind=Location)
    Secondary=(Weight=1,Axis=(X=0,Y=0,Z=0),Target=(X=1,Y=0,Z=0),Kind=Location)
    Weight=1
    bPropagateToChildren=True
    Primary.Target=&RigUnit_GetBoneTransform(Bone=head_wire_l).Transform.Translation
   )
   > &RigUnit_AimBone(
    Bone=head_wire_r
    DebugSettings=(bEnabled=false,Scale=10,WorldOffset=(Rotation=(X=0,Y=0,Z=0,W=1),Translation=(X=0,Y=0,Z=0),Scale3D=(X=1,Y=1,Z=1)))
    Primary=(Weight=1,Axis=(X=1,Y=0,Z=0),Target=(X=1,Y=0,Z=0),Kind=Location,Space="None")
    Secondary=(Weight=1,Axis=(X=0,Y=0,Z=0),Target=(X=1,Y=0,Z=0),Kind=Location)
    Weight=1
    bPropagateToChildren=True
    Primary.Target=&RigUnit_GetBoneTransform(Bone=neck_wire_r).Transform.Translation
   )
   > &RigUnit_AimBone(
    Bone=neck_wire_r
    DebugSettings=(Scale=10,WorldOffset=(Rotation=(X=0,Y=0,Z=0,W=1),Translation=(X=0,Y=0,Z=0),Scale3D=(X=1,Y=1,Z=1)))
    Primary=(Weight=1,Axis=(X=1,Y=0,Z=0),Target=(X=1,Y=0,Z=0),Kind=Location)
    Secondary=(Weight=1,Axis=(X=0,Y=0,Z=0),Target=(X=1,Y=0,Z=0),Kind=Location)
    Weight=1
    bPropagateToChildren=True
    Primary.Target=&RigUnit_GetBoneTransform(Bone=head_wire_r).Transform.Translation
   )
  ]}
]
hierarchy [
 root {loc=<0 0 0>, rot=<0 0 0 1>}
 pelvis {parent=root, loc=<0 -0.7071 93.1191>, rot=<0.0127 0.707 -0.0127 -0.707>}
 spine_01 {parent=pelvis, loc=<6.7494 -0.6388 0>, rot=<0 0 0.034 -0.9994>}
 spine_02 {parent=spine_01, loc=<7.2518 -0.035 0>, rot=<0 0 0.0238 -0.9997>}
 spine_03 {parent=spine_02, loc=<6.0104 0 0>, rot=<0 0 0.0149 -0.9999>}
 spine_04 {parent=spine_03, loc=<8.1024 0 0>, rot=<0 0 -0.0132 -0.9999>}
 spine_05 {parent=spine_04, loc=<10.0146 0 0>, rot=<0 0 -0.0015 -1>}
 clavicle_l {parent=spine_05, loc=<4.3575 -0.1466 -2.3223>, rot=<0.0465 -0.7289 -0.009 -0.6829>}
 clavicle_r {parent=spine_05, loc=<4.3575 -0.1466 2.3223>, rot=<0.7289 0.0465 -0.6829 0.009>}
 neck_01 {parent=spine_05, loc=<11.3138 0.1529 0>, rot=<0 0 -0.2003 -0.9797>}
 neck_02 {parent=neck_01, loc=<4.5627 0 0>, rot=<0 0 0.0504 -0.9987>}
 head {parent=neck_02, loc=<5.3914 0 0>, rot=<0 0 0.115 -0.9934>}
 dyn_parent_head_rod_l {parent=head, loc=<2.4017 -1.6086 -4.071>, rot=<-0.0602 0.6866 0.7244 -0.0178>}
 head_rod_l {parent=dyn_parent_head_rod_l, loc=<0 0 0>, rot=<0 0 0 1>}
 dyn_parent_head_rod_r {parent=head, loc=<2.4022 -1.6086 4.071>, rot=<0.0602 -0.6866 0.7244 -0.0178>}
 head_rod_r {parent=dyn_parent_head_rod_r, loc=<0 0 0>, rot=<0 0 0 -1>}
 dyn_parent_head_wire_l {parent=head, loc=<3.523 0.9039 -5.2942>, rot=<0.0782 -0.7499 -0.6449 0.125>}
 head_wire_l {parent=dyn_parent_head_wire_l, loc=<0 0 0>, rot=<0 0 0 -1>}
 dyn_parent_head_wire_r {parent=head, loc=<3.5234 0.9039 5.2942>, rot=<0.0782 -0.7499 0.6449 -0.125>}
 head_wire_r {parent=dyn_parent_head_wire_r, loc=<0 0 0>, rot=<0 0 0 -1>}
 faceAttach {parent=head, loc=<0.3389 -0.3281 -0.0071>, rot=<0.0082 -0.7071 -0.0082 -0.7071>, scale=<0.89 0.89 0.89>}
 C_brow_mid {parent=faceAttach, loc=<-0.0487 11.5818 5.3924>, rot=<0 0.0373 -0.0074 -0.9993>}
 L_brow_mid {parent=faceAttach, loc=<3.4282 11.4887 6.4258>, rot=<-0.0002 -0.0051 0.2553 -0.9669>}
 L_brow_outer {parent=faceAttach, loc=<6.068 9.1601 6.419>, rot=<0.0762 0.0921 0.4215 -0.8989>}
 L_cheek_inner {parent=faceAttach, loc=<4.197 9.6036 0.7663>, rot=<-0.0711 -0.2709 0.3182 -0.9057>}
 C_lip_upper_mid {parent=faceAttach, loc=<0.0119 11.9087 -2.1523>, rot=<0.2225 -0.0005 0.0002 -0.9749>}
 L_lip_upper_outer {parent=faceAttach, loc=<1.8042 11.0196 -1.9737>, rot=<0.0217 -0.1128 0.2775 -0.9538>}
 L_lip_corner {parent=faceAttach, loc=<2.5755 9.4918 -2.0116>, rot=<0.002 0.009 -0.0145 -0.9999>}
 C_nose_bridge {parent=faceAttach, loc=<-0.0412 11.7369 2.5806>, rot=<-0.2218 -0.0037 -0.001 -0.9751>}
 R_brow_mid {parent=faceAttach, loc=<-3.5028 11.4664 6.4406>, rot=<-0.0001 0.0043 -0.2553 -0.9669>}
 R_brow_outer {parent=faceAttach, loc=<-6.0879 9.146 6.4026>, rot=<0.0607 -0.0725 -0.4264 -0.8996>}
 R_cheek_inner {parent=faceAttach, loc=<-4.2468 9.5793 0.7545>, rot=<-0.0615 0.2468 -0.3215 -0.9121>}
 R_lip_corner {parent=faceAttach, loc=<-2.6089 9.5649 -1.9682>, rot=<0.0028 -0.0082 0.0146 -0.9999>}
 C_jaw {parent=faceAttach, loc=<0.0416 4.3097 1.3298>, rot=<0 0 0 1>}
 L_lip_lower_outer {parent=C_jaw, loc=<1.7194 6.6789 -3.6422>, rot=<0.0518 0.2564 0.4551 -0.8511>}
 C_lip_lower_mid {parent=C_jaw, loc=<-0.0539 7.3218 -4.0805>, rot=<0 0.0328 0.0069 -0.9994>}
 R_lip_lower_outer {parent=C_jaw, loc=<-1.8348 6.6561 -3.6513>, rot=<0.0647 -0.3269 -0.4997 -0.7995>}
 R_eye {parent=faceAttach, loc=<-3.2975 9.0746 4.7667>, rot=<0 0 0.9998 -0.0187>}
 L_eye {parent=faceAttach, loc=<3.2815 9.0746 4.7667>, rot=<0 0 0.9998 0.0187>}
 R_lip_upper_outer {parent=faceAttach, loc=<-1.7951 11.0753 -1.8924>, rot=<0.0191 0.1201 -0.2802 -0.9522>}
 dyn_parent_neck_rod_l {parent=spine_05, loc=<9.9561 -0.2988 -4.9148>, rot=<-0.7232 -0.0439 -0.0848 -0.684>}
 neck_rod_l {parent=dyn_parent_neck_rod_l, loc=<0 0 0>, rot=<0 0 0 1>}
 dyn_parent_neck_rod_r {parent=spine_05, loc=<9.956 -0.2988 4.9148>, rot=<0.7232 0.0439 -0.0848 -0.684>}
 neck_rod_r {parent=dyn_parent_neck_rod_r, loc=<0 0 0>, rot=<0 0 0 -1>}
 dyn_parent_neck_wire_l {parent=spine_05, loc=<9.5436 4.1153 -3.2177>, rot=<-0.7057 -0.0176 0.0616 -0.7056>}
 neck_wire_l {parent=dyn_parent_neck_wire_l, loc=<0 0 0>, rot=<0 0 0 1>}
 dyn_parent_neck_wire_r {parent=spine_05, loc=<9.5439 4.1153 3.2177>, rot=<0.7057 0.0176 0.0616 -0.7056>}
 neck_wire_r {parent=dyn_parent_neck_wire_r, loc=<0 0 0>, rot=<0 0 0 1>}
]
