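# Animation Blueprint for the Typhoon character's face accessory, following the attached parent mesh's pose and adjusting a neck wire.
# The copied pose and curves run through the accessory's Control Rig, then an additive bone-space yaw on the left neck wire bone.
# Each update derives that offset's alpha from head twist relative to spine_05 on the owner's body mesh.
# A second twist value, head_back, is computed but unused in the AnimGraph.
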AnimBlueprint /BRCosmetics/Characters/Player/Male/Medium/Bodies/M_MED_Typhoon/Meshes/Parts/M_MED_Typhoon_FaceAcc_AnimBP
target_skeleton=&/M_MED_Typhoon_FaceAcc_Skeleton
graphs [
 AnimGraph {nodes [
   n1 = "Transform (Modify) Bone - Bone: dyn_parent_neck_wire_l"(ComponentPose=Local To Component(), Alpha=1, Translation="0,0,0", Rotation="0,-15,0", Scale="1,1,1") {
    rotationMode=BMM_Additive
    rotationSpace=BCS_BoneSpace
    alphaBoolBlend {blendOption=Linear}
    alphaScaleBiasClamp {bInterpResult=true}
    boneToModify {boneName=dyn_parent_neck_wire_l}
    rotation {yaw=-15}
   }
   Component To Local(ComponentPose=@n1)
   n2 = Control Rig(Source=Copy Pose From Mesh() {bCopyCurves=true, bUseAttachedParent=true}) {
    alphaBoolBlend {blendOption=Linear}
    controlRigAssetReference {blueprintRigClass=&/M_MED_Typhoon_FaceAcc_CtrlRig.M_MED_Typhoon_FaceAcc_CtrlRig_C}
   }
   n3 = "Transform (Modify) Bone - Bone: dyn_parent_neck_wire_l"(
    ComponentPose=Local To Component(LocalPose=@n2)
    Alpha=Get neck_alpha() -> (neck_alpha:double)
    Translation="0,0,0"
    Rotation="0,-15,0"
    Scale="1,1,1"
   ) {
    rotationMode=BMM_Additive
    rotationSpace=BCS_BoneSpace
    alphaBoolBlend {blendOption=Linear}
    alphaScaleBiasClamp {bInterpResult=true}
    boneToModify {boneName=dyn_parent_neck_wire_l}
    rotation {yaw=-15}
   }
   Output Pose(Result=Component To Local(ComponentPose=@n3)) {layerGroup=None}
  ]}
 EventGraph {nodes [
   Event BlueprintInitializeAnimation()
   > Set Body_Mesh(Body_Mesh=GetOwnerSkeletalMeshComponent())
   > n1 = Cast To PlayerPawn_Athena(Object=TryGetPawnOwner())
   > n2 = Set Pawn Owner(Pawn Owner:Player Pawn Athena Object Reference=@n1) -> (Output_Get:Player Pawn Athena Object Reference)
   > Set Head_Mesh(Head_Mesh:Skeletal Mesh Component Object Reference=GetSkeletalMeshForPartType(self=@n2, PartType=Head)) -> (Output_Get:Skeletal Mesh Component Object Reference)
   Event BlueprintUpdateAnimation()
   > Get!(execute:Exec) -> (then:Exec, else:Exec, Body_Mesh:Skeletal Mesh Component Object Reference)
   > Sequence()
   > Set neck_alpha(neck_alpha:double=GetAlphaFromReferenceTransformDelta(
     SkeletalMeshComponent=Get Body_Mesh()
     BoneName=head
     BaseName=spine_05
     DeltaComponentType=QuaternionTwist
     InputRange=(X=0,Y=35)
     TwistAxis="0,1,0"
    )) -> (Output_Get:double)
   > Set head_back(head_back:double=GetAlphaFromReferenceTransformDelta(
     SkeletalMeshComponent=Get Body_Mesh()
     BoneName=head
     BaseName=spine_05
     DeltaComponentType=QuaternionTwist
     InputRange=(X=0,Y=48)
     TwistAxis="0,0,1"
    )) -> (Output_Get:double)
   TryGetPawnOwner()
   Event OnCharacterCustomizationLoaded()
  ]}
]
variables [
 Body_Mesh {type=SkeletalMeshComponent}
 head_back {type=number}
 Head_Mesh {type=SkeletalMeshComponent}
 neck_alpha {type=number}
 Pawn Owner {type=PlayerPawn_Athena_C}
]
